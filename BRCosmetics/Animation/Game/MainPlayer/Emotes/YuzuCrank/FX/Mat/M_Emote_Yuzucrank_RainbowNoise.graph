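# Translucent Niagara sprite material for the YuzuCrank emote, producing a scrolling iridescent rainbow noise effect.
# A panned holofoil texture is optionally darkened by power-sharpened panning noise (USEDARK switch), then tinted by particle color.
# Strength and desaturation, full by default, feed the emissive adjustments function; a glow mask with depth fade and particle alpha forms opacity.
# Both dark-noise speed axes share the parameter name Dark_Speed_X, so they cannot differ.
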Material /BRCosmetics/Animation/Game/MainPlayer/Emotes/YuzuCrank/FX/Mat/M_Emote_Yuzucrank_RainbowNoise
graph {
 properties {bAutomaticallySetUsageInEditor=false, bUsedWithNiagaraSprites=true, blendMode=BLEND_Translucent, usedShadingModels=Default Lit}
 nodes [
  n1 = TextureCoordinate()
  n2 = MaterialFunctionCall(X=ScalarParameter() {defaultValue=5, parameterName=UV_U}, Y=ScalarParameter() {defaultValue=5, parameterName=UV_V}) {materialFunction=&MakeFloat2}
  n3 = ScalarParameter() {parameterName=Speed_X}
  n4 = ScalarParameter() {defaultValue=10, parameterName=Speed_Y}
  n5 = TextureSample(UVs=Panner(
    Coordinate=Multiply(A=@n1, B=@n2.Result)
    Speed=MaterialFunctionCall(X=@n3, Y=@n4) {materialFunction=&MakeFloat2}.Result
   ) {speedY=5}) {
   samplerSource=SSM_Wrap_WorldGroupSettings
   samplerType=SAMPLERTYPE_LinearColor
   texture=/Game/Effects/Niagara/MasterMaterials/ShockWave/Texture/CA_Crew_Holofoil_Iridescent
  }
  n6 = TextureCoordinate()
  n7 = MaterialFunctionCall(X=ScalarParameter() {defaultValue=0.5, parameterName=Dark_Scale_X}, Y=ScalarParameter() {defaultValue=1, parameterName=Dark_Scale_Y}) {materialFunction=&MakeFloat2}
  n8 = ScalarParameter() {parameterName=Dark_Speed_X}
  n9 = ScalarParameter() {parameterName=Dark_Speed_X}
  n10 = TextureSample(UVs=Panner(
    Coordinate=Multiply(A=@n6, B=@n7.Result)
    Speed=MaterialFunctionCall(X=@n8, Y=@n9) {materialFunction=&MakeFloat2}.Result
   ) {speedX=0.5}) {
   samplerSource=SSM_Wrap_WorldGroupSettings
   samplerType=SAMPLERTYPE_LinearGrayscale
   texture=&/T_Gameplay_RadioTower_RedLights_Noise
  }
  n11 = Saturate(None=Power(Base=@n10.RGB, Exp=ScalarParameter() {defaultValue=5, parameterName=Dark_Stength}))
  n12 = ParticleColor()
  n13 = Multiply(A=StaticSwitch(True=Multiply(A=@n5.RGB, B=@n11), False=@n5.RGB, Value=StaticBoolParameter() {parameterName=USEDARK}), B=@n12.RGB)
  n14 = ScalarParameter() {defaultValue=1, parameterName=RainbowStrength}
  MaterialFunctionCall(Emissive=Desaturation(None=Multiply(A=@n13, B=@n14), Fraction=ScalarParameter() {defaultValue=1, parameterName=DesaturationStrength})) {materialFunction=&MF_FX_EmissiveAdjustments}
  Multiply(
   A=@n12.A
   B=DepthFade(
    Opacity=TextureSample() {texture=/BRCosmetics/Effects/Fort_Effects/Emotes/SaladDressing/FX/T_Emote_SaladDressing_glowmask}.R
    FadeDistance=ScalarParameter() {defaultValue=60, parameterName=FadeDistance}
   )
  )
 ]
}
